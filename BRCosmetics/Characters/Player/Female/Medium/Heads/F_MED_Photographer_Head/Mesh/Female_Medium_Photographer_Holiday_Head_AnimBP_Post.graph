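# Post-process Animation Blueprint for the Female Medium Photographer Holiday head that adjusts the head bone on the incoming pose.
# A one-time check on update sets a flag if the character item carries the PhotographerHoliday cosmetic meta tag.
# When flagged, each update stores a scaled-down copy of the head's rotation delta from ref pose, relative to neck_02, in Head Rot.
# The AnimGraph's Modify Bone replaces head rotation in bone space, with no visible link to Head Rot.
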AnimBlueprint /BRCosmetics/Characters/Player/Female/Medium/Heads/F_MED_Photographer_Head/Mesh/Female_Medium_Photographer_Holiday_Head_AnimBP_Post
target_skeleton=/Game/Characters/Player/Common/Fortnite_Base_Head/Fortnite_Base_Head_Export_Skeleton
graphs [
 AnimGraph {nodes [
   n1 = "Transform (Modify) Bone - Bone: head"!(ComponentPose=Local To Component(LocalPose=Input Pose()), bAlphaBoolEnabled:bool=True, Translation="0,0,0", Rotation="0,0,0", Scale="1,1,1") -> (Pose) {
    alphaInputType=Bool
    rotationMode=BMM_Replace
    rotationSpace=BCS_BoneSpace
    alphaBoolBlend {blendOption=Linear}
    boneToModify {boneName=head}
   }
   Output Pose(Result=Component To Local(ComponentPose=@n1)) {layerGroup=None}
  ]}
 EventGraph {nodes [
   Event BlueprintUpdateAnimation()
   > n1 = Sequence()
   > Do Once()
   > n2 = Branch(Condition=CharacterItemDefinitionHasAnyMetaTag(Tags=Make Array("[0]"=(TagName="Cosmetics.MetaTags.PhotographerHoliday"))))
   > Set IsPhotographerHoliday(IsPhotographerHoliday:bool=true) -> (Output_Get:bool)
   Set IsPhotographerHoliday(execute=@n2.else, IsPhotographerHoliday=false)
   Branch(execute=@n1.then_1, Condition=Get IsPhotographerHoliday() -> (IsPhotographerHoliday:bool))
   n3 = GetOwnerSkeletalMeshComponent()
   n4 = BreakRotator(InRot=BreakTransform(InTransform=GetDeltaTransformFromRefPose(self=@n3, BoneName=head, BaseName=neck_02)).Rotation)
   n5 = Get Head Rot Percentage() -> (Head Rot Percentage:double)
   > Set Head Rot(Head Rot:Rotator=MakeRotator(
     Roll=float * float(A=@n4, B=@n5) {is=title K2Node_PromotableOperator}
     Pitch=float * float(A=@n4.Pitch, B=@n5)
     Yaw=float * float(A=@n4.Yaw, B=@n5)
    )) -> (Output_Get:Rotator)
  ]}
]
variables [
 Head Rot {type=Rotator}
 Head Rot Percentage {type=number, default=0.3}
 IsPhotographerHoliday {type=boolean}
]
